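# Physics asset for the IvoryDwell backpack cosmetic mesh, giving collision bodies and joints to the pack, its head and two ears.
# The pack uses a box plus a small capsule, while the head and each two-segment dyn_ ear chain use capsules.
# Constraints chain each ear tip to its ear base, both bases to the head, and the head to the pack, all with matching swing and twist limit angles.
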PhysicsAsset /BRCosmetics/Accessories/FORT_Backpacks/Backpack_M_MED_IvoryDwell/Meshes/M_MED_IvoryDwell_Pack_Physics
bodies {
 dyn_ear_01_l [{shapeType=Capsule, length=3.5477, radius=2.9265, shapeName=ear_l_capsule, center=<2.4909 0 0>, rotation=<rot 0 90 90>}]
 dyn_ear_01_r [{shapeType=Capsule, length=3.5477, radius=2.9295, shapeName=ear_r_capsule, center=<2.4593 0 0>, rotation=<rot 0 -90 -90>}]
 dyn_ear_02_l [{shapeType=Capsule, length=2.4681, radius=1.6837, shapeName=earTip_l_capsule, center=<2.2376 0 0>, rotation=<rot 0 90 90>}]
 dyn_ear_02_r [{shapeType=Capsule, length=2.4681, radius=1.6842, shapeName=earTip_r_capsule, center=<2.2086 0 0>, rotation=<rot 0 -90 -90>}]
 pack [
  {shapeType=Capsule, length=0.505, radius=0.505, shapeName=pack_capsule, rotation=<rot 0 -90 0>}
  {shapeType=Box, extentX=22.1875, extentY=4.3125, extentZ=13.4375, shapeName=pack_box, center=<0 0 7.854>}
 ]
 pack_head [{shapeType=Capsule, length=2.5612, radius=5.2336, shapeName=head_capsule, center=<2.9774 0.0824 2.6533>, rotation=<rot -2.1637 -90 -14.4337>}]
}
constraints [
 {bone1Name=pack_head, bone2Name=pack, swing1LimitDegrees=45, swing2LimitDegrees=45, twistLimitDegrees=45}
 {bone1Name=dyn_ear_01_l, bone2Name=pack_head, swing1LimitDegrees=45, swing2LimitDegrees=45, twistLimitDegrees=45}
 {bone1Name=dyn_ear_02_l, bone2Name=dyn_ear_01_l, swing1LimitDegrees=45, swing2LimitDegrees=45, twistLimitDegrees=45}
 {bone1Name=dyn_ear_01_r, bone2Name=pack_head, swing1LimitDegrees=45, swing2LimitDegrees=45, twistLimitDegrees=45}
 {bone1Name=dyn_ear_02_r, bone2Name=dyn_ear_01_r, swing1LimitDegrees=45, swing2LimitDegrees=45, twistLimitDegrees=45}
]
